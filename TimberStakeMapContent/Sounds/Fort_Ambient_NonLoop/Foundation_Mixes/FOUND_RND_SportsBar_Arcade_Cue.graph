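# Sound cue named for random sports bar arcade ambience in the TimberStake map's Foundation mixes, with no nodes and no root, so it plays nothing as it stands.
# It routes to the Foundations_Ambience sound class at reduced volume and references the Fort_Map_3x3_Square attenuation asset.
# It also stores natural-sound falloff overrides with distance low-pass filtering.
# Its concurrency override prevents new instances once the limit is reached.
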SoundCue /TimberStakeMapContent/Sounds/Fort_Ambient_NonLoop/Foundation_Mixes/FOUND_RND_SportsBar_Arcade_Cue
root=null
nodes []
properties {
 attenuationSettings=/Game/Sounds/Attenuation/Fort_Map_3x3_Square
 bOverrideConcurrency=true
 soundClassObject=/TimberStakeMapContent/Sounds/FortSoundClassesAndModes/Foundations_Ambience
 volumeMultiplier=0.5
 attenuationOverrides {
  bAttenuateWithLPF=true
  distanceAlgorithm=NaturalSound
  falloffDistance=1500
  lPFRadiusMax=1500
  lPFRadiusMin=800
  nonSpatializedRadiusStart=100
 }
 concurrencyOverrides {resolutionRule=PreventNew}
}
